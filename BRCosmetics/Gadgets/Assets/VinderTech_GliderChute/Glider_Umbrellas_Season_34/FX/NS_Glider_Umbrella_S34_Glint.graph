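# Niagara system that adds golden glint sprites to the Season 34 umbrella glider.
# Flare_Bottom (GPU) and the lower-rate Flare_Low place sprites on a hard-coded dome of offsets from one skeletal mesh bone or socket.
# Flare_Top partitions particles across five bones or sockets with small random jitter.
# All emitters pulse size by curve, tint by time of day and use two sprite renderers, with a front-end preview flag feeding size.
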NiagaraSystem /BRCosmetics/Gadgets/Assets/VinderTech_GliderChute/Glider_Umbrellas_Season_34/FX/NS_Glider_Umbrella_S34_Glint
dependencies {
 usedDataInterfaces [&NiagaraDataInterfaceCurve, &NiagaraDataInterfaceArrayFloat3, &NiagaraDataInterfaceSkeletalMesh]
 usedDynamicInputs []
 usedModules [
  &SystemState
  &self:SystemSpawnScript.SystemScriptSource.SystemScriptGraph.NiagaraNodeAssignment_3.SetVariables_#1
 ]
}
emitters {
 Flare_Bottom {
  simTarget=GPUComputeSim
  emitterUpdateScript [
   &EmitterState(
    Delay First Loop Only=true
    Inactive Response=::NewEnumerator1
    Life Cycle Mode=::NewEnumerator1
    Loop Delay=0.8
    Loop Duration=1
    UseLoopDelay=true
   )
   &SpawnRate(SpawnRate=60, Use Spawn Probability=false)
  ]
  particleSpawnScript [
   &InitializeParticle(
    Color=<rgba 1 0.7501 0.1086 1>
    Lifetime Max=1
    Lifetime Min=0.5
    Lifetime Mode=::NewEnumerator1
    Sprite Size Mode=::NewEnumerator2
    UsePositionOffset=false
    Write Material Random=true
   )
   &ShapeLocation(
    Hemisphere Distribution=<1 0.5>
    MaskSpawnGroup=false
    Non Uniform Scale=<1 1 0.45>
    Offset=<0 0 7>
    Offset Mode=::NewEnumerator0
    Override Seed=false
    Sphere Radius=50
    Sphere Surface Distribution=0.8
    UseOwnerScale=true
   ) {module=ShapeLocation001}
   &CameraOffset(Camera Offset Amount=5)
  ]
  particleUpdateScript [
   &ParticleState()
   &ScaleSpriteSize(
    Uniform Curve Scale=&MakeCustomFloatFromBool(Bool=$User.bIsFrontEndPreview, True Float=4, False Float=5)
    Uniform Curve Sprite Scale=DataInterface() {propertyValues {Curve {keys [{tangentMode=RCTM_User, time=0, value=0}, <at 0.2 1>, <at 0.3 0.4>, {tangentMode=RCTM_User, time=1, value=0}]}}}
   )
   /SetVariables(Particles.Color002=&MultiplyLinearColor(
     Linear Color=$Particles.Initial.Color
     Scale Factor=&MakeFloat4FromVectorAndFloat(
      "Vector (RGB)"=&Multiply_Vector(
       A=<1 1 0.5>
       B=&FN_Select_Vector_from_Time_Of_Day(Morning=<1 1 1>, Day=<1 1 1>, Evening=<1 1 1>, Night=<0.5 0.5 1>)
      )
      "Float (Alpha)"=0.03
     )
    )) {isSetParameters=true, module=SetVariables_#2}
   &SkeletalMeshLocation(
    Bone / Socket Index=5
    Bone Sampling Mode=::NewEnumerator4
    MaskSpawnGroup=false
    Sampled Position Offset=&Subtract_Position(
     A=&SelectVectorFromArray(
      Array Sampling Mode=::NewEnumerator0
      Direct Array Index=?
      Vector Selection Array=DataInterface() {propertyValues {FloatData [
         <13.8925 -33.4166 108.5879>
         <-40.9037 -21.4009 101.0845>
         <-29.7756 -34.3399 101.0845>
         <16.0309 -50.8509 96.0936>
         <-30.3698 -21.1857 106.3789>
         <40.5543 -6.8035 106.0788>
         <44.7121 7.7969 102.3344>
         <-0.1862 27.0093 111.0892>
         <-37.2515 -37.9795 96.0936>
         <-16.5548 -8.559 115.9711>
         <-45.1147 -6.4851 101.6777>
         <9.5009 -45.7948 101.0845>
         <-52.0097 13.3947 96.0936>
         <7.5544 -16.7138 115.9197>
         <2.877 -39.2635 106.0799>
         <46.57 -22.2339 101.0847>
         <-29.8487 -9.7988 111.0914>
         <-0.4558 -10.119 120.9878>
         <4.5741 19.7334 117.2946>
         <-17.923 -39.2745 106.0781>
         <53.2346 2.0892 96.0936>
         <43.3628 22.5599 101.0846>
         <0.4626 -28.4254 112.1184>
         <12.4855 -4.6826 120.9995>
         <-4.0087 36.8952 106.0819>
         <-50.0684 -18.1599 96.0936>
         <34.7888 33.1211 101.0845>
         <31.1683 -16.362 111.0925>
         <43.4847 31.6437 96.0936>
         <-12.4365 -52.4965 96.0936>
         <34.8775 -26.9557 105.2266>
         <7.6343 39.6133 106.0793>
         <40.9239 -34.4685 96.0936>
         <-3.0232 9.04 120.9991>
         <19.8396 42.1652 101.0845>
         <50.5527 16.8423 96.0936>
         <-23.4689 18.4531 111.0908>
         <-34.1948 11.9012 109.3669>
         <15.5379 51.1751 96.0936>
         <-13.7231 51.8578 96.0936>
         <-39.0098 -13.2329 106.0796>
         <31.6608 43.6787 96.0936>
         <52.7352 -10.9806 97.107>
         <-42.7908 -30.3589 101.0847>
         <23.6022 32.8123 106.0789>
         <-46.5521 7.1513 101.0845>
         <34.5503 15.9056 106.0809>
         <27.311 -33.6394 101.0844>
         <30.0921 -44.8291 96.0936>
         <-7.2022 -40.6152 106.0786>
         <-13.4918 42.4212 101.0844>
         <10.7527 7.2399 120.9947>
         <-0.9676 49.9158 101.0846>
         <20.2473 -42.1401 101.0845>
         <27.7566 23.0132 108.8105>
         <-53.185 -2.0685 96.0936>
         <-25.7084 36.848 103.9396>
         <-37.3414 1.9944 106.0421>
         <-11.9713 13.1 120.8606>
         <-24.7348 -47.2974 96.0936>
         <-17.3951 31.5668 106.083>
         <-46.6048 26.5389 96.0936>
         <1.5364 -53.2184 96.0936>
         <-8.1878 -17.2689 115.9304>
         <-20.5227 -18.9523 112.2436>
         <24.8562 9.0525 115.3727>
         <-26.9089 0.6624 111.0892>
         <-11.088 25.132 114.7427>
         <-40.0873 21.5717 103.7965>
         <12.5711 29.2655 111.0916>
         <35.57 4.5388 109.9926>
         <-23.4047 -29.0309 107.3263>
         <26.4915 -4.8221 115.0444>
         <-11.7318 -28.402 111.0911>
         <18.111 20.0425 115.4662>
         <18.9721 -20.097 111.0892>
         <-31.9093 29.1666 101.0844>
         <-38.0625 37.1678 96.0936>
         <-27.4656 46.2428 96.0936>
         <-16.7405 3.4418 118.4196>
        ]}}
      Random Seed=0
     )
     B=<0 0 88>
    )
    Skeletal Mesh=$System.Skeletal Mesh
   )
   &ScaleColor(
    Scale RGB=&Multiply_Vector(
     A=&MakeVectorFromLinearColorRGB(Color=$Particles.Initial.Color)
     B=&FN_Select_Vector_from_Time_Of_Day(Morning=<0.75 0.75 1>, Day=<1 1 1>, Evening=<1 1 1>, Night=<0.1 0.1 0.5>)
    )
    ScaleA=true
    ScaleRGB=true
   )
  ]
  rendererClasses [&NiagaraSpriteRendererProperties]
  renderers [
   {rendererClass=&NiagaraSpriteRendererProperties}
   {rendererClass=&NiagaraSpriteRendererProperties, rendererIndex=1}
  ]
 }
 Flare_Low {
  emitterUpdateScript [
   &EmitterState(
    Delay First Loop Only=true
    Inactive Response=::NewEnumerator1
    Life Cycle Mode=::NewEnumerator1
    Loop Delay=0.8
    Loop Duration=1
    UseLoopDelay=true
   )
   &SpawnRate(SpawnRate=30, Use Spawn Probability=false)
  ]
  particleSpawnScript [
   &InitializeParticle(
    Color=<rgba 1 0.7501 0.1086 1>
    Lifetime Max=1
    Lifetime Min=0.5
    Lifetime Mode=::NewEnumerator1
    Sprite Size Mode=::NewEnumerator2
    UsePositionOffset=false
    Write Material Random=true
   )
   &ShapeLocation(
    Hemisphere Distribution=<1 0.5>
    MaskSpawnGroup=false
    Non Uniform Scale=<1 1 0.45>
    Offset=<0 0 7>
    Offset Mode=::NewEnumerator0
    Override Seed=false
    Sphere Radius=50
    Sphere Surface Distribution=0.8
    UseOwnerScale=true
   ) {module=ShapeLocation001}
   &CameraOffset(Camera Offset Amount=5)
  ]
  particleUpdateScript [
   &ParticleState()
   &ScaleSpriteSize(
    Uniform Curve Scale=&MakeCustomFloatFromBool(Bool=$User.bIsFrontEndPreview, True Float=4, False Float=5)
    Uniform Curve Sprite Scale=DataInterface() {propertyValues {Curve {keys [{tangentMode=RCTM_User, time=0, value=0}, <at 0.2 1>, <at 0.3 0.4>, {tangentMode=RCTM_User, time=1, value=0}]}}}
   )
   /SetVariables(Particles.Color002=&MultiplyLinearColor(
     Linear Color=$Particles.Initial.Color
     Scale Factor=&MakeFloat4FromVectorAndFloat(
      "Vector (RGB)"=&Multiply_Vector(
       A=<1 1 0.5>
       B=&FN_Select_Vector_from_Time_Of_Day(Morning=<1 1 1>, Day=<1 1 1>, Evening=<1 1 1>, Night=<0.5 0.5 1>)
      )
      "Float (Alpha)"=0.03
     )
    )) {isSetParameters=true, module=SetVariables_#2}
   &SkeletalMeshLocation(
    Bone / Socket Index=5
    Bone Sampling Mode=::NewEnumerator4
    MaskSpawnGroup=false
    Sampled Position Offset=&Subtract_Position(
     A=&SelectVectorFromArray(
      Array Sampling Mode=::NewEnumerator0
      Direct Array Index=?
      Vector Selection Array=DataInterface() {propertyValues {FloatData [
         <13.8925 -33.4166 108.5879>
         <-40.9037 -21.4009 101.0845>
         <-29.7756 -34.3399 101.0845>
         <16.0309 -50.8509 96.0936>
         <-30.3698 -21.1857 106.3789>
         <40.5543 -6.8035 106.0788>
         <44.7121 7.7969 102.3344>
         <-0.1862 27.0093 111.0892>
         <-37.2515 -37.9795 96.0936>
         <-16.5548 -8.559 115.9711>
         <-45.1147 -6.4851 101.6777>
         <9.5009 -45.7948 101.0845>
         <-52.0097 13.3947 96.0936>
         <7.5544 -16.7138 115.9197>
         <2.877 -39.2635 106.0799>
         <46.57 -22.2339 101.0847>
         <-29.8487 -9.7988 111.0914>
         <-0.4558 -10.119 120.9878>
         <4.5741 19.7334 117.2946>
         <-17.923 -39.2745 106.0781>
         <53.2346 2.0892 96.0936>
         <43.3628 22.5599 101.0846>
         <0.4626 -28.4254 112.1184>
         <12.4855 -4.6826 120.9995>
         <-4.0087 36.8952 106.0819>
         <-50.0684 -18.1599 96.0936>
         <34.7888 33.1211 101.0845>
         <31.1683 -16.362 111.0925>
         <43.4847 31.6437 96.0936>
         <-12.4365 -52.4965 96.0936>
         <34.8775 -26.9557 105.2266>
         <7.6343 39.6133 106.0793>
         <40.9239 -34.4685 96.0936>
         <-3.0232 9.04 120.9991>
         <19.8396 42.1652 101.0845>
         <50.5527 16.8423 96.0936>
         <-23.4689 18.4531 111.0908>
         <-34.1948 11.9012 109.3669>
         <15.5379 51.1751 96.0936>
         <-13.7231 51.8578 96.0936>
         <-39.0098 -13.2329 106.0796>
         <31.6608 43.6787 96.0936>
         <52.7352 -10.9806 97.107>
         <-42.7908 -30.3589 101.0847>
         <23.6022 32.8123 106.0789>
         <-46.5521 7.1513 101.0845>
         <34.5503 15.9056 106.0809>
         <27.311 -33.6394 101.0844>
         <30.0921 -44.8291 96.0936>
         <-7.2022 -40.6152 106.0786>
         <-13.4918 42.4212 101.0844>
         <10.7527 7.2399 120.9947>
         <-0.9676 49.9158 101.0846>
         <20.2473 -42.1401 101.0845>
         <27.7566 23.0132 108.8105>
         <-53.185 -2.0685 96.0936>
         <-25.7084 36.848 103.9396>
         <-37.3414 1.9944 106.0421>
         <-11.9713 13.1 120.8606>
         <-24.7348 -47.2974 96.0936>
         <-17.3951 31.5668 106.083>
         <-46.6048 26.5389 96.0936>
         <1.5364 -53.2184 96.0936>
         <-8.1878 -17.2689 115.9304>
         <-20.5227 -18.9523 112.2436>
         <24.8562 9.0525 115.3727>
         <-26.9089 0.6624 111.0892>
         <-11.088 25.132 114.7427>
         <-40.0873 21.5717 103.7965>
         <12.5711 29.2655 111.0916>
         <35.57 4.5388 109.9926>
         <-23.4047 -29.0309 107.3263>
         <26.4915 -4.8221 115.0444>
         <-11.7318 -28.402 111.0911>
         <18.111 20.0425 115.4662>
         <18.9721 -20.097 111.0892>
         <-31.9093 29.1666 101.0844>
         <-38.0625 37.1678 96.0936>
         <-27.4656 46.2428 96.0936>
         <-16.7405 3.4418 118.4196>
        ]}}
      Random Seed=0
     )
     B=<0 0 88>
    )
    Skeletal Mesh=$System.Skeletal Mesh
   )
   &ScaleColor(
    Scale RGB=&Multiply_Vector(
     A=&MakeVectorFromLinearColorRGB(Color=$Particles.Initial.Color)
     B=&FN_Select_Vector_from_Time_Of_Day(Morning=<0.75 0.75 1>, Day=<1 1 1>, Evening=<1 1 1>, Night=<0.1 0.1 0.5>)
    )
    ScaleA=true
    ScaleRGB=true
   )
  ]
  rendererClasses [&NiagaraSpriteRendererProperties]
  renderers [
   {rendererClass=&NiagaraSpriteRendererProperties}
   {rendererClass=&NiagaraSpriteRendererProperties, rendererIndex=1}
  ]
 }
 Flare_Top {
  emitterUpdateScript [
   &EmitterState(
    Delay First Loop Only=true
    Inactive Response=::NewEnumerator1
    Life Cycle Mode=::NewEnumerator1
    Loop Delay=0.8
    Loop Duration=1
    UseLoopDelay=true
   )
   &SpawnRate(SpawnRate=10, Use Spawn Probability=false)
  ]
  particleSpawnScript [
   &PartitionParticles(Number Of Partitions=5)
   &InitializeParticle(
    Color=<rgba 1 0.7501 0.1086 1>
    Lifetime Max=1
    Lifetime Min=0.5
    Lifetime Mode=::NewEnumerator1
    Sprite Size Max=<8 8>
    Sprite Size Mode=::NewEnumerator2
    UsePositionOffset=false
    Write Material Random=true
   )
   &CameraOffset(Camera Offset Amount=5)
  ]
  particleUpdateScript [
   &ParticleState()
   &ScaleSpriteSize(
    Uniform Curve Scale=&MakeCustomFloatFromBool(Bool=$User.bIsFrontEndPreview, True Float=4, False Float=4)
    Uniform Curve Sprite Scale=DataInterface() {propertyValues {Curve {keys [{tangentMode=RCTM_User, time=0, value=0}, <at 0.2 1>, <at 0.3 0.4>, {tangentMode=RCTM_User, time=1, value=0}]}}}
   )
   /SetVariables(Particles.Color002=&MultiplyLinearColor(
     Linear Color=$Particles.Initial.Color
     Scale Factor=&MakeFloat4FromVectorAndFloat(
      "Vector (RGB)"=&Multiply_Vector(
       A=<1 1 0.5>
       B=&FN_Select_Vector_from_Time_Of_Day(Morning=<1 1 1>, Day=<1 1 1>, Evening=<1 1 1>, Night=<0.5 0.5 1>)
      )
      "Float (Alpha)"=0.03
     )
    )) {isSetParameters=true}
   &SkeletalMeshLocation(
    Bone / Socket Index=$Particles.PartitionParticles.Partition
    Bone Sampling Mode=::NewEnumerator4
    MaskSpawnGroup=false
    Sampled Position Offset=&Add_Vector(
     A=&Add_Vector(A=<0 0 0>, B=&RandomRangeVector(Minimum=<1 2 2>, Maximum=<-1 -2 -2>))
     B=&SelectVectorFromArray(
      Direct Array Index=$Particles.PartitionParticles.Partition
      Vector Selection Array=DataInterface() {propertyValues {FloatData [<5 0 0>, <3.5 0 0>, <4 0 0>, <3 0 0>, <4 0 0>]}}
     )
    )
    Skeletal Mesh=$System.Skeletal Mesh
   )
   &ScaleColor(
    Scale RGB=&Multiply_Vector(
     A=&MakeVectorFromLinearColorRGB(Color=$Particles.Initial.Color)
     B=&FN_Select_Vector_from_Time_Of_Day(Morning=<0.75 0.75 1>, Day=<1 1 1>, Evening=<1 1 1>, Night=<0.1 0.1 0.5>)
    )
    ScaleA=true
    ScaleRGB=true
   )
  ]
  rendererClasses [&NiagaraSpriteRendererProperties]
  renderers [
   {rendererClass=&NiagaraSpriteRendererProperties}
   {rendererClass=&NiagaraSpriteRendererProperties, rendererIndex=1}
  ]
 }
}
summary {
 systemName=NS_Glider_Umbrella_S34_Glint
 emitters [
  {emitterName=Flare_Bottom, simTarget=GPUComputeSim, rendererClasses [&NiagaraSpriteRendererProperties]}
  {emitterName=Flare_Top, rendererClasses [&NiagaraSpriteRendererProperties]}
  {emitterName=Flare_Low, rendererClasses [&NiagaraSpriteRendererProperties]}
 ]
 userVariables [
  User.bIsFrontEnd {type=NiagaraBool}
  User.bIsFrontEndPreview {type=NiagaraBool}
  User.ForwardDot {type=NiagaraFloat, defaultValue=0, description=""}
  User.SystemDelay {type=NiagaraFloat, defaultValue=0, description=""}
 ]
}
